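# Sound cue for a balloon gadget impact, apparently against indestructible surfaces.
# A random node picks one of three equally weighted impact wave variations.
# It plays in the impacts sound class with boosted volume and gadget-activation attenuation.
# A per-owner concurrency rule caps it at two instances, stopping the oldest.
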SoundCue /Game/Sounds/Fort_Gadgets/Balloons/Balloon_Impact_Indest_Cue
root=Random_0
nodes [
 {id=Random_0, children [WavePlayer_1,WavePlayer_2,WavePlayer_3], properties {weights [1,1,1]}}
 {id=WavePlayer_1, properties {soundWaveAssetPtr=&/Balloon_Impact_Indestruct_01}}
 {id=WavePlayer_2, properties {soundWaveAssetPtr=&/Balloon_Impact_Indestruct_02}}
 {id=WavePlayer_3, properties {soundWaveAssetPtr=&/Balloon_Impact_Indestruct_03}}
]
properties {
 attenuationSettings=&Fort_GadgetActivate_Att
 maxDistance=4000
 soundClassObject=&Fort_Impacts
 volumeMultiplier=2
 concurrencySet [/Game/Sounds/Concurrency_Rules/TwoMaximum_Owner_StopOldest]
}
