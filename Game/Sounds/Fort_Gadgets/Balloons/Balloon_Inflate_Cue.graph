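# Sound cue for the balloon gadget's inflate sound.
# A Random node picks one of four equally weighted inflate wave variants.
# Playback runs at reduced volume with the in-air balloon attenuation, routed through the third-person ranged weapon fire sound class.
# A one-maximum-per-owner concurrency rule stops the oldest instance when a new one starts.
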SoundCue /Game/Sounds/Fort_Gadgets/Balloons/Balloon_Inflate_Cue
root=Random_0
nodes [
 {id=Random_0, children [WavePlayer_1,WavePlayer_2,WavePlayer_3,WavePlayer_4], properties {weights [1,1,1,1]}}
 {id=WavePlayer_1, properties {soundWaveAssetPtr=&/Balloon_Inflate_01}}
 {id=WavePlayer_2, properties {soundWaveAssetPtr=&/Balloon_Inflate_02}}
 {id=WavePlayer_3, properties {soundWaveAssetPtr=&/Balloon_Inflate_03}}
 {id=WavePlayer_4, properties {soundWaveAssetPtr=&/Balloon_Inflate_04}}
]
properties {
 attenuationSettings=/Game/Sounds/Attenuation/Fort_Balloon_InAir_Athena_Att
 maxDistance=8000
 soundClassObject=&Fort_WeaponRanged_Fire_3P_PSM
 volumeMultiplier=0.2
 concurrencySet [/Game/Sounds/Concurrency_Rules/OneMaximum_Owner_StopOldest]
}
