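# Sound cue for the looping balloon gadget sound tied to maximum altitude, routed to the Fort gadgets items sound class.
# A quality-level root sends the first two levels to the main looping wave through an enveloper that ramps volume up over time.
# The last two levels play an optimized looping wave directly, without the envelope.
# It references in-air balloon attenuation and a concurrency rule capping three per owner, preventing new ones.
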SoundCue /Game/Sounds/Fort_Gadgets/Balloons/Balloon_MaxAltitude_Cue
root=QualityLevel_0
nodes [
 {
  id=Enveloper_0
  children [WavePlayer_3]
  properties {pitchCurve {editorCurveData {keys [{time=0}]}}, volumeCurve {editorCurveData {keys [<at 0.0036 0.0666>,<at 1 1>]}}}
 }
 {id=QualityLevel_0, children [Enveloper_0,Enveloper_0,WavePlayer_4,WavePlayer_4]}
 {id=WavePlayer_3, properties {bLooping=true, soundWaveAssetPtr=&/Balloon_MaxAltitudeLoop}}
 {id=WavePlayer_4, properties {bLooping=true, soundWaveAssetPtr=&/Balloon_MaxAltitudeLoop_opt}}
]
properties {
 attenuationSettings=/Game/Sounds/Attenuation/Fort_Balloon_InAir_Athena_Att
 maxDistance=8000
 soundClassObject=&Fort_Gadgets_Items
 volumeMultiplier=0.3
 attenuationOverrides {bAttenuate=false}
 concurrencySet [/Game/Sounds/Concurrency_Rules/ThreeMaximum_Owner_PreventNew]
}
