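# Sound cue for the balloon gadget pickup sound, classed under Fort gadget items.
# A single Balloon_PickedUp wave plays through a modulator that slightly randomizes pitch for variety.
# It uses gadget-activate attenuation with a boosted volume, and a concurrency override that prevents new instances once the limit is reached.
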SoundCue /Game/Sounds/Fort_Gadgets/Balloons/Balloon_PickedUp_Cue
root=Modulator_0
nodes [
 {id=Modulator_0, children [WavePlayer_5], properties {pitchMax=1.1, pitchMin=0.9}}
 {id=WavePlayer_5, properties {soundWaveAssetPtr=&/Balloon_PickedUp}}
]
properties {
 attenuationSettings=&Fort_GadgetActivate_Att
 bOverrideConcurrency=true
 maxDistance=4000
 soundClassObject=&Fort_Gadgets_Items
 volumeMultiplier=1.9
 concurrencyOverrides {resolutionRule=PreventNew}
}
