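# Sound cue for the anniversary birthday variant of a balloon pop gadget sound, classed as an impact.
# A Random node picks one of three equally weighted pop waves and feeds a Modulator that varies pitch upward only, at constant volume.
# The cue uses the Athena balloon pop attenuation and boosts overall volume.
# An enabled concurrency override keeps few simultaneous pops by stopping the oldest, although a shared impacts concurrency set is also assigned.
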SoundCue /Game/Sounds/Fort_Gadgets/Balloons/Balloon_Pop_Anniversary_Cue
root=Modulator_7
nodes [
 {id=Modulator_7, children [Random_2], properties {pitchMax=1.2, pitchMin=1, volumeMax=1, volumeMin=1}}
 {id=Random_2, children [WavePlayer_0,WavePlayer_4,WavePlayer_8], properties {bShouldExcludeFromBranchCulling=true, weights [1,1,1]}}
 {id=WavePlayer_0, properties {soundWaveAssetPtr=&/Balloon_Anniversary_Birthday_Pop_03}}
 {id=WavePlayer_4, properties {soundWaveAssetPtr=&/Balloon_Anniversary_Birthday_Pop_01}}
 {id=WavePlayer_8, properties {soundWaveAssetPtr=&/Balloon_Anniversary_Birthday_Pop_02}}
]
properties {
 attenuationSettings=/Game/Sounds/Attenuation/Fort_Balloon_Pop_Athena_Att
 bOverrideConcurrency=true
 maxDistance=4500
 soundClassObject=&Fort_Impacts
 volumeMultiplier=1.6
 concurrencyOverrides {maxCount=2, resolutionRule=StopOldest}
 concurrencySet [&CONC_Object_Impacts_FourMax]
}
