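# Sound cue for the pop heard when a balloon gadget is shot.
# A Random node picks one of three ShotPop wave variations with equal weight.
# It uses an in-air balloon attenuation asset and a third-person ranged weapon fire sound class.
# A concurrency rule, named for a maximum of two per owner that stops the oldest, limits overlapping plays.
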SoundCue /Game/Sounds/Fort_Gadgets/Balloons/Balloon_Shot_Cue
root=Random_1
nodes [
 {id=Random_1, children [WavePlayer_0,WavePlayer_5,WavePlayer_6], properties {weights [1,1,1]}}
 {id=WavePlayer_0, properties {soundWaveAssetPtr=&/Balloon_ShotPop_1}}
 {id=WavePlayer_5, properties {soundWaveAssetPtr=&/Balloon_ShotPop_3}}
 {id=WavePlayer_6, properties {soundWaveAssetPtr=&/Balloon_ShotPop_2}}
]
properties {
 attenuationSettings=/Game/Sounds/Attenuation/Fort_Balloon_InAir_Athena_Att
 maxDistance=8000
 soundClassObject=&Fort_WeaponRanged_Fire_3P_PSM
 volumeMultiplier=1
 concurrencySet [/Game/Sounds/Concurrency_Rules/TwoMaximum_Owner_StopOldest]
}
