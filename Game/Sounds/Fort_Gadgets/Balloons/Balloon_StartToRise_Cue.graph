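# Sound cue for the balloon gadget's start-to-rise sound.
# A quality level node routes its first two levels to the standard wave and its last two to an "_opt" variant.
# It uses an in-air balloon attenuation and borrows the third-person ranged weapon fire sound class.
# Concurrency follows a shared rule named for two maximum per owner, stopping the oldest.
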SoundCue /Game/Sounds/Fort_Gadgets/Balloons/Balloon_StartToRise_Cue
root=QualityLevel_0
nodes [
 {id=QualityLevel_0, children [WavePlayer_0,WavePlayer_0,WavePlayer_1,WavePlayer_1]}
 {id=WavePlayer_0, properties {soundWaveAssetPtr=&/Balloon_StartToRise}}
 {id=WavePlayer_1, properties {soundWaveAssetPtr=&/Balloon_StartToRise_opt}}
]
properties {
 attenuationSettings=/Game/Sounds/Attenuation/Fort_Balloon_InAir_Athena_Att
 maxDistance=8000
 soundClassObject=&Fort_WeaponRanged_Fire_3P_PSM
 concurrencySet [/Game/Sounds/Concurrency_Rules/TwoMaximum_Owner_StopOldest]
}
