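# Sound cue for the looping sound of the balloon gadget floating in the air.
# A quality-level node sends its first two tiers to an equally weighted random pick between two looping waves, and its last two to a single optimized loop.
# It uses the balloon in-air Athena attenuation asset, the gadget items sound class, and a concurrency rule named for three maximum per owner, preventing new sounds.
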SoundCue /Game/Sounds/Fort_Gadgets/Balloons/Baloon_FloatingLoop_Cue
root=QualityLevel_0
nodes [
 {id=QualityLevel_0, children [Random_0,Random_0,WavePlayer_0,WavePlayer_0]}
 {id=Random_0, children [WavePlayer_1,WavePlayer_2], properties {weights [1,1]}}
 {id=WavePlayer_0, properties {bLooping=true, soundWaveAssetPtr=&/Baloon_FloatingLoop_opt}}
 {id=WavePlayer_1, properties {bLooping=true, soundWaveAssetPtr=&/Baloon_FloatingLoop_01}}
 {id=WavePlayer_2, properties {bLooping=true, soundWaveAssetPtr=&/Baloon_FloatingLoop_02}}
]
properties {
 attenuationSettings=/Game/Sounds/Attenuation/Fort_Balloon_InAir_Athena_Att
 maxDistance=8000
 soundClassObject=&Fort_Gadgets_Items
 volumeMultiplier=1
 attenuationOverrides {attenuationShape=Capsule, falloffDistance=5000, reverbDistanceMin=1000, reverbWetLevelMax=3, attenuationShapeExtents {x=4000}}
 concurrencySet [/Game/Sounds/Concurrency_Rules/ThreeMaximum_Owner_PreventNew]
}
