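# Sound cue for the foley heard when a balloon gadget attaches.
# A single wave player feeds the Baloon_Foley_Attach wave straight to the output, with no randomization or modulation.
# It uses the in-air balloon attenuation, the gadget items sound class, and a per-owner concurrency rule that caps it at two instances, stopping the oldest.
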SoundCue /Game/Sounds/Fort_Gadgets/Balloons/Baloon_Foley_Attach_Cue
root=WavePlayer_0
nodes [{id=WavePlayer_0, properties {soundWaveAssetPtr=&/Baloon_Foley_Attach}}]
properties {
 attenuationSettings=/Game/Sounds/Attenuation/Fort_Balloon_InAir_Athena_Att
 maxDistance=8000
 soundClassObject=&Fort_Gadgets_Items
 concurrencySet [/Game/Sounds/Concurrency_Rules/TwoMaximum_Owner_StopOldest]
}
